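# Animation sequence for the PistolCanary creature's single-tentacle area-of-effect attack, with three spawned actors.
# The tentacle skeletal mesh holds a fixed transform and carries a two-section Control Rig track, whose listed channels are all constant.
# A sphere mesh grows from near zero, arcs high and plunges three times.
# At each plunge a cylinder mesh rises and rapidly widens horizontally, then snaps back, apparently marking the impact area.
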LevelSequence /PistolCanary/Anim/SingleTentacle/Section02/PistolCanary_Attack_AOE/PistolCanary_TentacleAttack_AOE_anim
bindings [
 Cylinder {
  id=#1
  possessed_class=FortStaticMeshActor
  tracks [
   {type=MovieSceneSpawnTrack, sections [{type=MovieSceneSpawnSection, channels [None {type=MovieSceneScriptingBoolChannel, default=true}]}]}
   {
    type=MovieScene3DTransformTrack
    sections [{
      type=MovieScene3DTransformSection
      channels [
       Location.X {type=MovieSceneScriptingDoubleChannel, default=13132.1514, const [13132.1514,12]}
       Location.Y {type=MovieSceneScriptingDoubleChannel, default=68914.4646, const [68914.4609,12]}
       Location.Z {
        type=MovieSceneScriptingDoubleChannel
        default=2515
        keys [[3699,2150.1633],[3700,2300.7981],[3704,2515],[3705,2150.1633],[3744,2150.1633],[3745,2313.0718],[3749,2515],[3750,2150.1633],[3786,2150.1633],[3787,2322.1562],[3793,2515],[3794,2150.1633]]
       }
       Rotation.X {type=MovieSceneScriptingDoubleChannel, default=0, const [0,12]}
       Rotation.Y {type=MovieSceneScriptingDoubleChannel, default=0, const [0,12]}
       Rotation.Z {type=MovieSceneScriptingDoubleChannel, default=0, const [0,12]}
       Scale.X {
        type=MovieSceneScriptingDoubleChannel
        default=500
        keys [[3699,1],[3700,207.0284],[3704,500],[3705,1],[3744,1],[3745,223.8157],[3749,500],[3750,1],[3786,1],[3787,236.2407],[3793,500],[3794,1]]
       }
       Scale.Y {
        type=MovieSceneScriptingDoubleChannel
        default=500
        keys [[3699,1],[3700,207.0284],[3704,500],[3705,1],[3744,1],[3745,223.8157],[3749,500],[3750,1],[3786,1],[3787,236.2407],[3793,500],[3794,1]]
       }
       Scale.Z {type=MovieSceneScriptingDoubleChannel, default=1, const [1,12]}
      ]
     }]
   }
  ]
 }
 Sphere {
  id=#2
  possessed_class=FortStaticMeshActor
  tracks [
   {
    type=MovieSceneSpawnTrack
    sections [{type=MovieSceneSpawnSection, channels [None {type=MovieSceneScriptingBoolChannel, default=true, const [true,1]}]}]
   }
   {
    type=MovieScene3DTransformTrack
    sections [{
      type=MovieScene3DTransformSection
      channels [
       Location.X {
        type=MovieSceneScriptingDoubleChannel
        default=17320.3104
        thinned_from=28
        keys [[3608,14343.3633],[3623,13976.0576],[3634,14055.1895],[3648,13853.7734],[3667,13660.9027],[3686,13636.0283],[3692,14343.3633],[3734,14343.3633],[3738,14128.0332],[3741,13863.7705],[3744,13731.6377],[3749,14343.3633],[3759,15101.1064],[3772,14454.9424],[3784,14343.3633],[3792,14343.3633]]
       }
       Location.Y {
        type=MovieSceneScriptingDoubleChannel
        default=69866.3337
        thinned_from=31
        keys [[3608,70454.1797],[3623,71711.6016],[3634,71463.7891],[3648,69554.4952],[3667,68618.1622],[3686,68215.2344],[3692,70408.1719],[3697,75500.4453],[3700,73730.1797],[3701,72733.7188],[3704,72733.7188],[3705,70454.1797],[3718,68835.8984],[3734,69856.9609],[3738,71523.3672],[3741,75244.5703],[3744,74762.0938],[3745,72733.7188],[3748,72733.7188],[3749,70454.1797],[3759,68175.3125],[3772,69640.1562],[3777,70326.1947],[3784,73700.2109],[3787,73730.1797],[3788,72733.7188],[3791,72733.7188],[3792,70454.1797]]
       }
       Location.Z {
        type=MovieSceneScriptingDoubleChannel
        default=0
        thinned_from=31
        keys [[3608,2292.0994],[3634,8027.9023],[3648,12618.9551],[3667,17147.2424],[3686,17972.3535],[3692,18424.0488],[3697,13388.9613],[3700,5783.8509],[3701,4591.397],[3702,-1509.1298],[3704,-2036.8654],[3705,2292.0994],[3718,11693.5098],[3734,17413.7207],[3738,17241.3496],[3741,12935.9326],[3744,7479.9712],[3745,4591.397],[3746,-1509.1298],[3748,-2036.8654],[3749,2292.0994],[3759,11790.0449],[3772,17452.8691],[3777,16290.0473],[3784,12818.9111],[3787,8615.4498],[3788,4591.397],[3789,-1509.1298],[3791,-2036.8654],[3792,-14639.0723]]
       }
       Rotation.X {type=MovieSceneScriptingDoubleChannel, default=0, const [0,26]}
       Rotation.Y {type=MovieSceneScriptingDoubleChannel, default=0, const [0,26]}
       Rotation.Z {type=MovieSceneScriptingDoubleChannel, default=0, const [0,26]}
       Scale.X {
        type=MovieSceneScriptingDoubleChannel
        default=0.01
        thinned_from=27
        keys [[3607,0.01],[3608,40],[3634,32.2744],[3648,26.0907],[3667,19.9915],[3686,18.8801],[3692,18.2717],[3697,18.357],[3700,18.4556],[3701,49.4208],[3702,40],[3718,22.814],[3734,18.8801],[3741,18.435],[3744,18.4556],[3745,49.4208],[3746,40],[3759,28.7376],[3772,19.6884],[3777,19.3346],[3784,18.4213],[3787,18.4556],[3788,49.4208],[3792,49.4208]]
       }
       Scale.Y {
        type=MovieSceneScriptingDoubleChannel
        default=0.01
        thinned_from=27
        keys [[3607,0.01],[3608,40],[3634,32.2744],[3648,26.0907],[3667,19.9915],[3686,18.8801],[3692,18.2717],[3697,18.357],[3700,18.4556],[3701,49.4208],[3702,40],[3718,22.814],[3734,18.8801],[3741,18.435],[3744,18.4556],[3745,49.4208],[3746,40],[3759,28.7376],[3772,19.6884],[3777,19.3346],[3784,18.4213],[3787,18.4556],[3788,49.4208],[3792,49.4208]]
       }
       Scale.Z {
        type=MovieSceneScriptingDoubleChannel
        default=0.01
        thinned_from=27
        keys [[3607,0.01],[3608,40],[3634,32.2744],[3648,26.0907],[3667,19.9915],[3686,18.8801],[3692,18.2717],[3697,18.357],[3700,18.4556],[3701,49.4208],[3702,40],[3718,22.814],[3734,18.8801],[3741,18.435],[3744,18.4556],[3745,49.4208],[3746,40],[3759,28.7376],[3772,19.6884],[3777,19.3346],[3784,18.4213],[3787,18.4556],[3788,49.4208],[3792,49.4208]]
       }
      ]
     }]
   }
  ]
 }
 Tentacle Pistol Canary Creature {
  id=#3
  possessed_class=SkeletalMeshActor
  tracks [
   {type=MovieSceneSpawnTrack, sections [{type=MovieSceneSpawnSection, channels [None {type=MovieSceneScriptingBoolChannel, default=true}]}]}
   {
    type=MovieSceneControlRigParameterTrack
    display_name=Tentacle_PistolCanary_Creature_CtrlRig
    sections [
     {
      type=MovieSceneControlRigParameterSection
      channels [
       control rig options.Location.X {type=MovieSceneScriptingFloatChannel, default=0, const [0,154]}
       control rig options.Location.Y {type=MovieSceneScriptingFloatChannel, default=0, const [0,154]}
       control rig options.Location.Z {type=MovieSceneScriptingFloatChannel, default=0, const [0,154]}
       control rig options.Rotation.X {type=MovieSceneScriptingFloatChannel, default=0, const [0,154]}
       control rig options.Rotation.Y {type=MovieSceneScriptingFloatChannel, default=0, const [0,154]}
       control rig options.Rotation.Z {type=MovieSceneScriptingFloatChannel, default=0, const [0,154]}
       control rig options.Scale.X {type=MovieSceneScriptingFloatChannel, default=1, const [1,154]}
       control rig options.Scale.Y {type=MovieSceneScriptingFloatChannel, default=1, const [1,154]}
       control rig options.Scale.Z {type=MovieSceneScriptingFloatChannel, default=1, const [1,154]}
       Weight {type=MovieSceneScriptingFloatChannel, default=1, const [1,8]}
      ]
     }
     {
      type=MovieSceneControlRigParameterSection
      channels [
       control rig options.Location.X {type=MovieSceneScriptingFloatChannel, default=0, const [0,8]}
       control rig options.Location.Y {type=MovieSceneScriptingFloatChannel, default=0, const [0,8]}
       control rig options.Location.Z {type=MovieSceneScriptingFloatChannel, default=0, const [0,8]}
       control rig options.Rotation.X {type=MovieSceneScriptingFloatChannel, default=0, const [0,8]}
       control rig options.Rotation.Y {type=MovieSceneScriptingFloatChannel, default=0, const [0,8]}
       control rig options.Rotation.Z {type=MovieSceneScriptingFloatChannel, default=0, const [0,8]}
       control rig options.Scale.X {type=MovieSceneScriptingFloatChannel, default=0, const [0,8]}
       control rig options.Scale.Y {type=MovieSceneScriptingFloatChannel, default=0, const [0,8]}
       control rig options.Scale.Z {type=MovieSceneScriptingFloatChannel, default=0, const [0,8]}
      ]
     }
    ]
   }
   {
    type=MovieScene3DTransformTrack
    sections [{
      type=MovieScene3DTransformSection
      channels [
       Location.X {type=MovieSceneScriptingDoubleChannel, default=0, const [14287.208,2]}
       Location.Y {type=MovieSceneScriptingDoubleChannel, default=0, const [70618.9219,2]}
       Location.Z {type=MovieSceneScriptingDoubleChannel, default=0, const [3947.8015,2]}
       Rotation.X {type=MovieSceneScriptingDoubleChannel, default=0, const [0,2]}
       Rotation.Y {type=MovieSceneScriptingDoubleChannel, default=0, const [0,2]}
       Rotation.Z {type=MovieSceneScriptingDoubleChannel, default=0, const [0,2]}
       Scale.X {type=MovieSceneScriptingDoubleChannel, default=1, const [1,2]}
       Scale.Y {type=MovieSceneScriptingDoubleChannel, default=1, const [1,2]}
       Scale.Z {type=MovieSceneScriptingDoubleChannel, default=1, const [1,2]}
      ]
     }]
   }
  ]
 }
]
display_rate {denominator=1, numerator=30}
tick_resolution {denominator=1, numerator=24000}
tracks []
